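; Hierarchical-generation PCG sample that scatters a forest of trees, boulders and seedling grass over world raycast surfaces.
; Trees spawn on 64m cells, and their widened extents are subtracted from boulders sampled on 32m cells.
; On 8m cells, grass density combines spatial noise with a boost near trees, scales the instances, then is threshold-filtered before spawning.
; A separate debug branch displays cell bounds for each grid level, including unbounded.
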(graph HiGenGraph
  ;[0 0.006406 0.1042] Default Unbounded Cell - Run Once
  (bind PointFromMesh_3 (Point%20From%20Mesh :Mesh /PCG/SampleContent/SimpleForest/Meshes/PCG_Boulder_02.PCG_Boulder_02 :MeshPathAttributeName Mesh))
  (bind GetAttributeFromPointIndex_9 (Get%20Attribute%20From%20Point%20Index :bForceOutputAttributeToBeInElementsDomain true :InputSource $Extents :OutputAttributeName @SourceName))
  (bind WorldRayHitQuery_4 (World%20Ray%20Hit%20Query :QueryParams (struct :bOverrideDefaultParams true)))
  (bind GetAttributeFromPointIndex_5 (Get%20Attribute%20From%20Point%20Index :bForceOutputAttributeToBeInElementsDomain true :InputSource Mesh :OutputAttributeName @SourceName))
  (bind PointFromMesh_16 (Point%20From%20Mesh :Mesh /PCG/SampleContent/SimpleForest/Meshes/PCG_Tree_01.PCG_Tree_01 :MeshPathAttributeName Mesh))
  (bind GetAttributeFromPointIndex_18 (Get%20Attribute%20From%20Point%20Index :bForceOutputAttributeToBeInElementsDomain true :InputSource $Extents :OutputAttributeName @SourceName))
  (bind WorldRayHitQuery_10 (World%20Ray%20Hit%20Query :QueryParams (struct :bOverrideDefaultParams true)))
  (bind GetAttributeFromPointIndex_20 (Get%20Attribute%20From%20Point%20Index :bForceOutputAttributeToBeInElementsDomain true :InputSource Mesh :OutputAttributeName @SourceName))
  ;[0.0625 0.03697 0] Rocks - 32m cells
  (bind HiGenGridSize_11 (Change%20Grid%20Size :HiGenGridSize Grid32))
  (bind SurfaceSampler_0 (Surface%20Sampler :PointExtents (1e+02 1e+02 1e+02) :PointSteepness 0.5))
  (bind ExtentsModifier_70 (Extents%20Modifier :Extents (3e+01 3e+01 5e+02)))
  (bind Difference_27 (Difference :bDiffMetadata false :DensityFunction Binary))
  (bind ToPoint_4 (To%20Point))
  (bind TransformPoints_2 (Transform%20Points :OffsetMin (0.0 0.0 -5e+01) :RotationMax (struct :Yaw 3.6e+02) :ScaleMax (0.8 1.0 1.0)))
  (bind AddAttribute_7 (Add%20Attribute :InputSource Mesh :OutputTarget @SourceName))
  (bind StaticMeshSpawner_2 (Static%20Mesh%20Spawner :MeshSelectorParameters (class:/Script/PCG.PCGMeshSelectorByAttribute :AttributeName Mesh :TemplateDescriptor (struct :ComponentClass /Script/Engine.HierarchicalInstancedStaticMeshComponent :Mobility Movable :bGenerateOverlapEvents true)) :MeshSelectorType /Script/PCG.PCGMeshSelectorByAttribute))
  ;[0.00427 0.03125 0] Trees - 64m cells
  (bind HiGenGridSize_12 (Change%20Grid%20Size :HiGenGridSize Grid64))
  (bind SurfaceSampler_14 (Surface%20Sampler :PointExtents (1e+02 1e+02 1e+02) :PointSteepness 0.5))
  (bind TransformPoints_26 (Transform%20Points :RotationMax (struct :Yaw 3.6e+02)))
  (bind AddAttribute_22 (Add%20Attribute :InputSource Mesh :OutputTarget @SourceName))
  (bind StaticMeshSpawner_24 (Static%20Mesh%20Spawner :MeshSelectorParameters (class:/Script/PCG.PCGMeshSelectorByAttribute :AttributeName Mesh :TemplateDescriptor (struct :ComponentClass /Script/Engine.HierarchicalInstancedStaticMeshComponent :Mobility Movable :bGenerateOverlapEvents true)) :MeshSelectorType /Script/PCG.PCGMeshSelectorByAttribute))
  ;[0 0.006406 0.1042] Default Unbounded Cell - Run Once
  (bind WorldRayHitQuery_19 (World%20Ray%20Hit%20Query :QueryParams (struct :bOverrideDefaultParams true :bIgnorePCGHits true)))
  ;[0.08854 0.0166 0.0166] Grass On Ground - 8m cells
  (bind HiGenGridSize_17 (Change%20Grid%20Size :HiGenGridSize Grid8))
  (bind SurfaceSampler_21 (Surface%20Sampler :PointExtents (1.5 1.5 2.0) :PointsPerSquaredMeter 1.5e+02 :PointSteepness 0.5))
  ;[0.0902 0.01569 0.01569] Grass on rocks and Ground: Spatial Noise -> Density and Scale - 8m Cells
  (bind Merge_8 (Merge%20Points :bMergeMetadata false))
  (bind Spatial%20Noise_4 (Spatial%20Noise :Brightness 0.05 :Contrast 8.0 :Iterations 2 :RandomOffset (1e+02 1e+02 1e+05) :Transform (struct :Scale3D (2e+02 2.5e+02 2e+03))))
  (bind AttributeMathsOp_3 (Attribute%20Maths%20Op :InputSource1 $Density :InputSource2 $Density))
  (bind AttributeMathsOp_7 (Attribute%20Maths%20Op :InputSource1 $Scale :InputSource2 $Density :Operation Multiply))
  (bind PointFilter_17 (Filter%20Attribute%20Elements :AttributeTypes (struct :DoubleValue 0.4) :bUseConstantThreshold true :bUseSpatialQuery true))
  ;;[0.0902 0.01569 0.01569] more points around trees
  (bind Distance_2 (Distance :bOutputToAttribute false :bSetDensity true :MaximumDistance 5e+02 :SourceShape Center :TargetShape Center))
  (bind AttributeMathsOp_4 (Attribute%20Maths%20Op :InputSource1 $Density :Operation OneMinus))
  (bind CreateAttribute_20 (Create%20Constant :AttributeTypes (struct :DoubleValue 3.0)))
  (bind AttributeMathsOp_18 (Attribute%20Maths%20Op :InputSource1 $Density :InputSource2 @Last :Operation Pow))
  (bind CreateAttribute_6 (Create%20Constant :AttributeTypes (struct :DoubleValue 2.0)))
  (bind AttributeMathsOp_5 (Attribute%20Maths%20Op :InputSource1 $Density :Operation Multiply :OutputTarget $Density))
  ;[0.0902 0.01569 0.01569] 8m Cells - Spawn grass
  (bind TransformPoints_23 (Transform%20Points :RotationMax (struct :Yaw 3.6e+02) :ScaleMax (0.2 1.0 1.0) :ScaleMin (0.1 1.0 1.0)))
  (bind StaticMeshSpawner_13 (Static%20Mesh%20Spawner :bApplyMeshBoundsToPoints false :MeshSelectorParameters (class:/Script/PCG.PCGMeshSelectorWeighted :MeshEntries [(struct :Descriptor (struct :StaticMesh /PCG/SampleContent/SimpleForest/Meshes/PCG_Seedling_01.PCG_Seedling_01 :OverrideMaterials [/PCG/SampleContent/SimpleForest/Materials/PCG_Foliage_01_Inst_wpo.PCG_Foliage_01_Inst_wpo] :ComponentClass /Script/Engine.HierarchicalInstancedStaticMeshComponent :Mobility Movable :BodyInstance (struct :VelocitySolverIterationCount 1 :CollisionEnabled NoCollision :CollisionProfileName NoCollision :CollisionResponses (struct :ResponseArray [(struct :Channel Visibility :Response ECR_Ignore) (struct :Channel Camera :Response ECR_Ignore)]) :bAutoWeld true) :InstanceEndCullDistance 1600 :bGenerateOverlapEvents true :WorldPositionOffsetDisableDistance 5000))])))
  ;[0 0 0] Debug Display: Grid Levels
  (bind CreatePoints_0 (Create%20Points :PointsToCreate []))
  (bind HiGenGridSize_26 (Change%20Grid%20Size :HiGenGridSize Unbounded))
  (bind HiGenGridSize_9 (Change%20Grid%20Size :HiGenGridSize Grid64))
  (bind Subgraph_0 (Subgraph :Graph /PCG/Subgraphs/SpatialDataBoundsToPoint.SpatialDataBoundsToPoint))
  (bind ExtentsModifier_29 (Extents%20Modifier :Extents (99999999.0 99999999.0 14.0) :Mode Minimum))
  (bind Debug_1 (Debug))
  (bind HiGenGridSize_13 (Change%20Grid%20Size :HiGenGridSize Grid32))
  (bind Subgraph_14 (Subgraph :Graph /PCG/Subgraphs/SpatialDataBoundsToPoint.SpatialDataBoundsToPoint))
  (bind ExtentsModifier_31 (Extents%20Modifier :Extents (99999999.0 99999999.0 14.5) :Mode Minimum))
  (bind Debug_16 (Debug))
  (bind HiGenGridSize_21 (Change%20Grid%20Size :HiGenGridSize Grid8))
  (bind Subgraph_17 (Subgraph :Graph /PCG/Subgraphs/SpatialDataBoundsToPoint.SpatialDataBoundsToPoint))
  (bind ExtentsModifier_33 (Extents%20Modifier :Extents (99999999.0 99999999.0 15.0) :Mode Minimum))
  (bind Debug_19 (Debug))
  (bind Subgraph_22 (Subgraph :Graph /PCG/Subgraphs/SpatialDataBoundsToPoint.SpatialDataBoundsToPoint))
  (bind ExtentsModifier_27 (Extents%20Modifier :Extents (99999999.0 99999999.0 13.5) :Mode Minimum))
  (bind Debug_24 (Debug))
  (connect GetAttributeFromPointIndex_5 Attribute AddAttribute_7 Attributes)
  (connect TransformPoints_2 Out AddAttribute_7 In)
  (connect GetAttributeFromPointIndex_20 Attribute AddAttribute_22 Attributes)
  (connect TransformPoints_26 Out AddAttribute_22 In)
  (connect Spatial%20Noise_4 Out AttributeMathsOp_3 InA)
  (connect AttributeMathsOp_5 Out AttributeMathsOp_3 InB)
  (connect Distance_2 Out AttributeMathsOp_4 In)
  (connect AttributeMathsOp_18 Out AttributeMathsOp_5 InA)
  (connect CreateAttribute_6 Out AttributeMathsOp_5 InB)
  (connect AttributeMathsOp_3 Out AttributeMathsOp_7 InA)
  (connect AttributeMathsOp_3 Out AttributeMathsOp_7 InB)
  (connect AttributeMathsOp_4 Out AttributeMathsOp_18 InA)
  (connect CreateAttribute_20 Out AttributeMathsOp_18 InB)
  (connect ExtentsModifier_29 Out Debug_1 In)
  (connect ExtentsModifier_31 Out Debug_16 In)
  (connect ExtentsModifier_33 Out Debug_19 In)
  (connect ExtentsModifier_27 Out Debug_24 In)
  (connect ExtentsModifier_70 Out Difference_27 Differences)
  (connect SurfaceSampler_0 Out Difference_27 Source)
  (connect Spatial%20Noise_4 Out Distance_2 Source)
  (connect TransformPoints_26 Out Distance_2 Target)
  (connect Subgraph_22 Point ExtentsModifier_27 In)
  (connect Subgraph_0 Point ExtentsModifier_29 In)
  (connect Subgraph_14 Point ExtentsModifier_31 In)
  (connect Subgraph_17 Point ExtentsModifier_33 In)
  (connect SurfaceSampler_14 Out ExtentsModifier_70 In)
  (connect PointFromMesh_3 Out GetAttributeFromPointIndex_5 In)
  (connect PointFromMesh_3 Out GetAttributeFromPointIndex_9 In)
  (connect PointFromMesh_16 Out GetAttributeFromPointIndex_18 In)
  (connect PointFromMesh_16 Out GetAttributeFromPointIndex_20 In)
  (connect HiGenGridSize_26 Out HiGenGridSize_9 In)
  (connect WorldRayHitQuery_4 Out HiGenGridSize_11 In)
  (connect WorldRayHitQuery_10 Out HiGenGridSize_12 In)
  (connect HiGenGridSize_9 Out HiGenGridSize_13 In)
  (connect WorldRayHitQuery_19 Out HiGenGridSize_17 In)
  (connect HiGenGridSize_13 Out HiGenGridSize_21 In)
  (connect CreatePoints_0 Out HiGenGridSize_26 In)
  (connect SurfaceSampler_21 Out Merge_8 In)
  (connect AttributeMathsOp_7 Out PointFilter_17 In)
  (connect Merge_8 Out Spatial%20Noise_4 In)
  (connect AddAttribute_7 Out StaticMeshSpawner_2 In)
  (connect TransformPoints_23 Out StaticMeshSpawner_13 In)
  (connect AddAttribute_22 Out StaticMeshSpawner_24 In)
  (connect HiGenGridSize_9 Grid%20Cell%20Volume Subgraph_0 SpatialData)
  (connect HiGenGridSize_13 Grid%20Cell%20Volume Subgraph_14 SpatialData)
  (connect HiGenGridSize_21 Grid%20Cell%20Volume Subgraph_17 SpatialData)
  (connect HiGenGridSize_26 Grid%20Cell%20Volume Subgraph_22 SpatialData)
  (connect HiGenGridSize_11 Grid%20Cell%20Volume SurfaceSampler_0 Bounding%20Shape)
  (connect GetAttributeFromPointIndex_9 Attribute SurfaceSampler_0 PointExtents)
  (connect HiGenGridSize_11 Out SurfaceSampler_0 Surface)
  (connect HiGenGridSize_12 Grid%20Cell%20Volume SurfaceSampler_14 Bounding%20Shape)
  (connect GetAttributeFromPointIndex_18 Attribute SurfaceSampler_14 PointExtents)
  (connect HiGenGridSize_12 Out SurfaceSampler_14 Surface)
  (connect HiGenGridSize_17 Grid%20Cell%20Volume SurfaceSampler_21 Bounding%20Shape)
  (connect HiGenGridSize_17 Out SurfaceSampler_21 Surface)
  (connect Difference_27 Out ToPoint_4 In)
  (connect ToPoint_4 Out TransformPoints_2 In)
  (connect PointFilter_17 InsideFilter TransformPoints_23 In)
  (connect SurfaceSampler_14 Out TransformPoints_26 In))
